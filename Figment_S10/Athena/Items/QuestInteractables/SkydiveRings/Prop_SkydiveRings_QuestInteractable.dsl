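; Skydive ring quest prop that credits a quest objective when a player passes through it outside Creative mode.
; BeginPlay colors and scales the ring by height level, then waits for playlist data before checking Creative mode and setting course difficulty.
; Clients hide rings based on quest ownership, after landing, or after a hit with success effects.
; A repeating timer tilts the mesh halfway between straight up and the local player above it.
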(event EventBeginPlay
  (bind _heightscalemulitplier (Variables|Default|GetHeightScaleMulitplier))
  (bind _scalemultiplier (Variables|RingScale|GetScaleMultiplier))
  (bind _returnvalue (* _scalemultiplier _heightscalemulitplier))
  (bind _sm_athena_skydivering (Variables|Default|GetSM_Athena_SkydiveRing))
  (bind _relativescale3d (Variables|/Script/Engine.SceneComponent|GetRelativeScale3D _sm_athena_skydivering))
  (bind _returnvalue_2 (* _relativescale3d _returnvalue))
  (bind _highestlevel (Variables|Default|GetHighestLevel))
  (bind _ringmat (Variables|Default|GetRingMat))
  (bind _middlelevel (Variables|Default|GetMiddleLevel))
  (bind _lowestlevel (Variables|Default|GetLowestLevel))
  (bind _p_athena_skydiverings (Variables|Default|GetP_Athena_SkydiveRings))
  ;[1 1 1] Set up Material
  (bind _returnvalue_3 (Rendering|Material|CreateDynamicMaterialInstance (Variables|Default|GetSM_Athena_SkydiveRing) 0 "/Game/Athena/Items/QuestInteractables/SkydiveRings/FX/M_Athena_SkydiveRings_Simple.M_Athena_SkydiveRings_Simple" "None" :!class /Script/Engine.PrimitiveComponent))
  (Variables|Default|SetRingMat _returnvalue_3)
  (Utilities|FlowControl|Sequence
    (:then_0
      ;[1 1 1] Hotfix Size Adjustments
      (bind _outxy (Fortnite|EvaluateCurveTableRow (Variables|RingScale|GetSizeScalerRow)))
      (Variables|RingScale|SetScaleMultiplier _outxy)
      ;[1 1 1] Check Height Level
      (switch int (Variables|Default|GetHeightLevel)
        (:0
          ;[1 1 1] Set Color according to Height Level
          (Rendering|Material|SetVectorParameterValue _ringmat "ColorInner" _lowestlevel :!class /Script/Engine.MaterialInstanceDynamic)
          (Effects|Components|ParticleSystem|SetColorParameter _p_athena_skydiverings "PulseRingColor" _lowestlevel)
          (Variables|Default|SetHeightScaleMulitplier 1.0))
        (:1
          (Rendering|Material|SetVectorParameterValue _ringmat "ColorInner" _middlelevel :!class /Script/Engine.MaterialInstanceDynamic)
          (Effects|Components|ParticleSystem|SetColorParameter _p_athena_skydiverings "PulseRingColor" _middlelevel)
          (Variables|Default|SetHeightScaleMulitplier 1.25))
        (:2
          (Rendering|Material|SetVectorParameterValue _ringmat "ColorInner" _highestlevel :!class /Script/Engine.MaterialInstanceDynamic)
          (Effects|Components|ParticleSystem|SetColorParameter _p_athena_skydiverings "PulseRingColor" _highestlevel)
          (Variables|Default|SetHeightScaleMulitplier 1.5)))
      ;[1 1 1] Set Scale according to HeightLevel
      (Variables|RingScale|SetAdjustedRingScale _returnvalue_2)
      (Transformation|SetRelativeScale3D _sm_athena_skydivering _returnvalue_2)
      (Components|Sphere|SetSphereRadius (Variables|Default|GetOverlapSphere) (* _returnvalue (Components|Sphere|GetScaledSphereRadius (Variables|Default|GetOverlapSphere))) true))
    (:then_1
      ;[1 0.6341 0.1593] Need Something Better from Code
      (Utilities|FlowControl|Delay 5.0)
      ;[1 1 1] CLIENT AND SERVER - Check if In Creative Mode or not
      (bind _asfort_game_state_athena (Utilities|Casting|CastToFortGameStateAthena (Game|GetGameState)))
      (if (Variables|/Script/FortniteGame.FortGameStateAthena|GetbPlaylistDataIsLoaded _asfort_game_state_athena)
        (bind _cancompletequest (Class|AthenaQuestFunctionLibrary|AthenaCheckCreativeMode _asfort_game_state_athena))
        (if _cancompletequest
          (Variables|Default|SetCreativeMode false)
          (else
            (Variables|Default|SetCreativeMode true)))
        ;[1 1 1] If skydiving course, set ring height level to scale difficulty
        (if (not (Variables|Default|GetIsSkydivingCourse))
          _
          (elif (Variables|Default|GetIsEasyCourse)
            (Variables|Default|SetHeightLevel 2)
            (elif (Variables|Default|GetIsHardCourse)
              (Variables|Default|SetHeightLevel 0)
              (else
                (Development|PrintString "Prop_SkydiveRings_QuestInteractable: Is set to SkydivingCourse but no difficulty is set")))))
        ;[1 1 1] Does Player have quest?
        (Networking|SwitchHasAuthority
          (:Remote
            (bind _isvalid (CallFunction|IsValidObjective))
            (CallFunction|HideSkydiveCourse)
            (if (and _isvalid (not (Variables|Default|GetCreativeMode)))
              _
              (elif (Variables|Default|GettestingEnabled)
                _
                (else
                  ;[1 1 1] Hide Ring when landing after skydive or if quest is invalid
                  (Development|PrintString "OBJ not vaild" (Variables|Default|GetDebugOn) (Variables|Default|GetDebugOn))
                  (CallFunction|HideRing))))
            ;[1 1 1] Bind Skydive and RingTriggered Events
            (bind _asathena_player_controller (Utilities|Casting|CastToAthena_PlayerController (Utilities|Array|Get(acopy) (Fortnite|GetLocalFortPlayerControllers true))))
            (CallFunction|BeginSkyDiving)
            (EventDispatchers|BindEventtoOnEndSkydiving _asathena_player_controller (EventDispatchers|CreateEvent :!event Custom|LandedFromSkydive))))
        (else
          (Playlist|BindEventtoOnCurrentPlaylistLoaded _asfort_game_state_athena (delegate Custom|PlaylistLoaded)))))))

(event Custom|PlaylistLoaded
  (params (PlaylistName :type Name) (PlaylistContextTags :type Gameplay%20Tag%20Container%20Structure%20%28by%20ref%29))
  ;[1 1 1] CLIENT AND SERVER - Check if In Creative Mode or not
  (bind _cancompletequest (Class|AthenaQuestFunctionLibrary|AthenaCheckCreativeMode (Game|GetGameState)))
  (if _cancompletequest
    (Variables|Default|SetCreativeMode false)
    (else
      (Variables|Default|SetCreativeMode true)))
  ;[1 1 1] If skydiving course, set ring height level to scale difficulty
  (if (not (Variables|Default|GetIsSkydivingCourse))
    _
    (elif (Variables|Default|GetIsEasyCourse)
      (Variables|Default|SetHeightLevel 2)
      (elif (Variables|Default|GetIsHardCourse)
        (Variables|Default|SetHeightLevel 0)
        (else
          (Development|PrintString "Prop_SkydiveRings_QuestInteractable: Is set to SkydivingCourse but no difficulty is set")))))
  ;[1 1 1] Does Player have quest?
  (Networking|SwitchHasAuthority
    (:Remote
      (bind _isvalid (CallFunction|IsValidObjective))
      (CallFunction|HideSkydiveCourse)
      (if (and _isvalid (not (Variables|Default|GetCreativeMode)))
        _
        (elif (Variables|Default|GettestingEnabled)
          _
          (else
            ;[1 1 1] Hide Ring when landing after skydive or if quest is invalid
            (Development|PrintString "OBJ not vaild" (Variables|Default|GetDebugOn) (Variables|Default|GetDebugOn))
            (CallFunction|HideRing))))
      ;[1 1 1] Bind Skydive and RingTriggered Events
      (bind _asathena_player_controller (Utilities|Casting|CastToAthena_PlayerController (Utilities|Array|Get(acopy) (Fortnite|GetLocalFortPlayerControllers true))))
      (CallFunction|BeginSkyDiving)
      (EventDispatchers|BindEventtoOnEndSkydiving _asathena_player_controller (EventDispatchers|CreateEvent :!event Custom|LandedFromSkydive)))))

(event Custom|BeginSkyDiving
  ;[1 1 1] Set functions to pulse and tilt toward player while skydiving
  (bind _returnvalue (Utilities|Time|SetTimerbyFunctionName 0 "OrientToLocalPlayer" 0.1 true))
  (Variables|Timers|SetOrientToPlayerTimer _returnvalue)
  (Development|PrintString "BEGIN SKYDIVE" (Variables|Default|GetDebugOn) (Variables|Default|GetDebugOn) "(R=0.170675,G=1.000000,B=0.000000,A=1.000000)" 7.0))

(event Custom|LandedFromSkydive
  ;[1 1 1] Hide Ring when landing after skydive or if quest is invalid
  (CallFunction|HideRing))

(event Custom|HideSkydiveCourse
  ;[1 1 1] Hide skydive inactive skydive courses
  (if (and (Variables|Default|GetIsEasyCourse) (not (Variables|Default|GetHasQuest1)))
    _
    (elif (and (Variables|Default|GetIsHardCourse) (not (Variables|Default|GetHasQuest2)))
      _))
  ;[1 1 1] Hide Ring when landing after skydive or if quest is invalid
  (if (Variables|Default|GettestingEnabled)
    ;[1 1 1] Bind Skydive and RingTriggered Events
    (bind _asathena_player_controller (Utilities|Casting|CastToAthena_PlayerController (Utilities|Array|Get(acopy) (Fortnite|GetLocalFortPlayerControllers true))))
    (CallFunction|BeginSkyDiving)
    (EventDispatchers|BindEventtoOnEndSkydiving _asathena_player_controller (EventDispatchers|CreateEvent :!event Custom|LandedFromSkydive))
    (else
      (Development|PrintString "OBJ not vaild" (Variables|Default|GetDebugOn) (Variables|Default|GetDebugOn))
      (CallFunction|HideRing))))

(event Collision|EventActorBeginOverlap
  (params (OtherActor :type Actor%20Object%20Reference))
  (bind _ishardcourse (Variables|Default|GetIsHardCourse))
  ;[1 1 1] On overlap
  (bind _asplayer_pawn_athena (Utilities|Casting|CastToPlayerPawn_Athena OtherActor))
  (if (not (Variables|Default|GetCreativeMode))
    (bind _asathena_player_controller (Utilities|Casting|CastToAthena_PlayerController (Game|GetInstigatorController OtherActor))
      (:then
        (if (not (Variables|Default|GetIsSkydivingCourse))
          (Quest|GiveQuestUpdatetoPlayers (Variables|QuestStuff|GetQuestItem1) (Variables|QuestStuff|GetQuestObjectiveBackendName1) (Variables|QuestStuff|GetRingsObjectiveStat) (Utilities|Array|MakeArray _asathena_player_controller))
          (else
            (Development|PrintString (Utilities|String|ToString(Boolean) _ishardcourse))
            ;;[1 1 1] Handle skydive course quest update
            (Quest|GiveQuestUpdatetoPlayers (select _ishardcourse (Variables|QuestStuff|GetQuestItem2) (Variables|QuestStuff|GetQuestItem1)) (select (Variables|Default|GetIsHardCourse) (Variables|QuestStuff|GetQuestObjectiveBackendName2) (Variables|QuestStuff|GetQuestObjectiveBackendName1)) (Variables|QuestStuff|GetRingsObjectiveStat) (Utilities|Array|MakeArray _asathena_player_controller))))
        ;[1 1 1] Client Only
        (Networking|SwitchHasAuthority
          (:Remote
            (Development|PrintString "Ring Triggered" (Variables|Default|GetDebugOn) (Variables|Default|GetDebugOn))
            ;[1 1 1] FX for Success
            (Audio|PlaySoundatLocation (Variables|Default|GetSuccessSound))
            (Effects|SpawnEmitterAttached (Variables|Default|GetSuccessEmitter) (Variables|/Script/Engine.Character|GetMesh _asplayer_pawn_athena))
            (Components|Activation|Activate (Variables|Default|GetForceFeedback) false)
            (CallFunction|HideRing))))
      (:CastFailed
        (Development|PrintString "CAST FAILED: SKYDIVE RINGS")))))

(event Custom|HideRing
  ;[1 1 1] Hide Ring
  (Rendering|SetActorHiddenInGame true)
  (Collision|SetActorEnableCollision false)
  (Components|Activation|Deactivate (Variables|Default|GetP_Athena_SkydiveRings))
  (Utilities|Time|ClearandInvalidateTimerbyHandle (Variables|Timers|GetOrientToPlayerTimer)))

(fn IsValidObjective
  (bind _returnvalue (Fortnite|GetLocalFortPlayerControllers true))
  (bind _output (Utilities|Array|Get(acopy) _returnvalue))
  (bind _questvalid (Quest|DoesPlayerHaveQuest _output (Variables|QuestStuff|GetQuestItem1) (Variables|QuestStuff|GetQuestObjectiveBackendName1)))
  (bind _questvalid_2 (Quest|DoesPlayerHaveQuest _output (Variables|QuestStuff|GetQuestItem2) (Variables|QuestStuff|GetQuestObjectiveBackendName2)))
  (Variables|Default|SetHasQuest1 _questvalid)
  (Variables|Default|SetHasQuest2 _questvalid_2)
  (return (or _questvalid_2 _questvalid)))

(fn OrientToLocalPlayer
  (bind _returnvalue (Fortnite|GetLocalFortPlayerControllers true))
  (bind _returnvalue_2 (Transformation|GetActorLocation))
  (bind _output (Utilities|Array|Get(acopy) _returnvalue))
  (bind _returnvalue_3 (Transformation|GetActorLocation _output))
  (Networking|SwitchHasAuthority
    (:Remote
      (if (> (.z _returnvalue_3) (.z _returnvalue_2))
        (Transformation|SetWorldRotation (Variables|Default|GetSM_Athena_SkydiveRing) (Math|Rotator|MakeRotfromZ (/ (+ (Variables|Default|GetStraightUp) (Math|Vector|Normalize (- _returnvalue_3 _returnvalue_2))) 2.0)) false false)))))

(fn ConstructionScript)
